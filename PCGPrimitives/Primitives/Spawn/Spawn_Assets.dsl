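; PCG primitive graph for spawning assets, built as a templated pipeline with swappable stage subgraphs.
; Primary input is tagged "Input:0", then passes through input management (fallback and processing graphs), core process and output preparation.
; Core process receives offsets, seeded random transform ranges, scale options, mesh or index attribute selection, and a material override.
; Graph parameters supply the stage subgraphs, allowing replacement without editing the pipeline.
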(graph Spawn_Assets
  (params (InputsProcessingGraph :type (SoftObject /Script/PCG.PCGGraphInterface) :default /PCGPrimitives/Primitives/_Template_/Primitives_InputsProcessing.Primitives_InputsProcessing) (InputsFallbackGraph :type (SoftObject /Script/PCG.PCGGraphInterface) :default /PCGPrimitives/Primitives/Spawn/Subgraphs/Spawn_Assets_InputsFallback.Spawn_Assets_InputsFallback) (CoreProcessGraph :type (SoftObject /Script/PCG.PCGGraphInterface) :default /PCGPrimitives/Primitives/Spawn/Subgraphs/Spawn_Assets_CoreProcess.Spawn_Assets_CoreProcess) (PrepareOutputsGraph :type (SoftObject /Script/PCG.PCGGraphInterface) :default /PCGPrimitives/Primitives/_Template_/Primitives_PrepareOutputs.Primitives_PrepareOutputs) (Assets :type (Array (SoftObject /Script/CoreUObject.Object))) (PositionOffset :type Vector) (RotationOffset :type Rotator) (ScaleOffset :type Vector :default (1.0 1.0 1.0)) (UniformScale :type Bool :default true) (MatchScaleToInputBounds :type Bool) (RandomPositionRange :type Vector) (RandomRotationRange :type Vector) (RandomScaleRange :type Vector) (seed :type Int) (UseMeshAttribute :type Bool) (MeshAttribute :type String) (UseAssetIndexAttribute :type Bool) (IndexAttribute :type String) (MaterialOverride :type (SoftObject /Script/Engine.MaterialInterface)))
  (inputs (PrimaryInput :type Any))
  (bind AddTags_0 (Add%20Tags :TagsToAdd "Input:0"))
  ;[0 0 0 0.5 32] Inputs
  (bind Subgraph_0 (Subgraph :Graph /PCGPrimitives/Primitives/_Template_/Subgraphs/Primitives_InputsManagement.Primitives_InputsManagement :InputsFallbackGraph InputsFallbackGraph :InputsProcessingGraph InputsProcessingGraph))
  ;[0 0 0 0.5 32] Core Process
  (bind Subgraph_3 (Subgraph :Graph /PCGPrimitives/Primitives/Spawn/Subgraphs/Spawn_Assets_CoreProcess.Spawn_Assets_CoreProcess :Index_attribute IndexAttribute :Match_scale_to_input_bounds MatchScaleToInputBounds :Mesh_Attrinute MeshAttribute :OverlayMaterial MaterialOverride :PositionOffset PositionOffset :Random_Position_Range RandomPositionRange :Random_Rotation_Range RandomRotationRange :Random_Scale_Range RandomScaleRange :RotationOffset RotationOffset :ScaleOffset ScaleOffset :seed-param seed :SubgraphOverride-param CoreProcessGraph :UniformScale UniformScale :Use_Index_attribute UseAssetIndexAttribute :Use_Mesh_Attribute UseMeshAttribute))
  ;[0 0 0 0.5 32] Prepare Output
  (bind Subgraph_4 (Subgraph :Graph /PCGPrimitives/Primitives/_Template_/Primitives_PrepareOutputs.Primitives_PrepareOutputs :SubgraphOverride-param PrepareOutputsGraph))
  (connect DefaultInputNode PrimaryInput AddTags_0 In)
  (connect AddTags_0 Out Subgraph_0 In)
  (connect Subgraph_0 Out Subgraph_3 In)
  (connect Subgraph_3 Out Subgraph_4 In)
  (output Out Subgraph_4))
